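# Animation Blueprint for the CoolSlice Golf backpack cosmetic, adding physics motion to its book, chain and handle bones.
# Serial RootRelative AnimDynamics nodes simulate the chain from a cached 'start pose', with spherical limits on the left upper arm and spine.
# An isMale blend selects between two chains with differently sized limit spheres.
# Another chain variant is much stiffer, with tighter angular limits, reduced component motion and spherical limits disabled.
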AnimBlueprint /BRCosmetics/Accessories/FORT_Backpacks/Backpack_M_MED_CoolSlice_Golf/Meshes/M_MED_CoolSlice_Golf_Pack_AnimBP
target_skeleton=&/M_MED_CoolSlice_Golf_Pack_Skeleton
graphs [
 AnimGraph {nodes [
   Anim Dynamics - Bone: dyn_book!(ComponentPose, Alpha=1, GravityOverride:Vector="0,0,-980") -> (Pose) {
    angularDampingOverride=1
    angularSpringConstant=150
    bLinearSpring=true
    bOverrideAngularDamping=true
    bOverrideLinearDamping=true
    bUseGravityOverride=true
    linearDampingOverride=1
    linearSpringConstant=50
    simulationSpace=RootRelative
    boundBone {boneName=dyn_book}
    componentAppliedLinearAccClamp=<1100 1100 1100>
    componentLinearAccScale=<0.5 0.5 0.5>
    componentLinearVelScale=<0.5 0.5 0.5>
    physicsBodyDefinitions [{
      boundBone {boneName=dyn_book}
      boxExtents=<100 100 100>
      constraintSetup {angularLimitsMax=<0 0 0>, angularLimitsMin=<0 0 0>, angularTarget=<1 0 0>, linearAxesMax=<3.7 0 0>}
      localJointOffset=<1 0 0>
     }]
   }
   n1 = Anim Dynamics - Bone: dyn_chain_01!(ComponentPose=Local To Component(LocalPose=Use cached pose 'start pose'()), Alpha=1, GravityOverride:Vector="0,0,-980") -> (Pose) {
    angularDampingOverride=0.9
    angularSpringConstant=5000
    bOverrideAngularDamping=true
    bUseGravityOverride=true
    bUseSphericalLimits=true
    linearDampingOverride=1
    linearSpringConstant=500
    simulationSpace=RootRelative
    boundBone {boneName=dyn_chain_01}
    componentAppliedLinearAccClamp=<305 305 305>
    componentLinearAccScale=<0.9 0.9 0.9>
    componentLinearVelScale=<0.9 0.9 0.9>
    physicsBodyDefinitions [{
      collisionType=CustomSphere
      sphereCollisionRadius=1.5
      boundBone {boneName=dyn_chain_01}
      boxExtents=<37 27 27>
      constraintSetup {angularLimitsMax=<0 0 15>, angularLimitsMin=<0 -15 -15>, angularTarget=<1 0 0>}
      localJointOffset=<5 0 0>
     }]
    sphericalLimits [
     {limitRadius=10, drivingBone {boneName=upperarm_l_in}, sphereLocalOffset=<0 0 0>}
     {limitRadius=17, drivingBone {boneName=spine_05_in}, sphereLocalOffset=<0 0 0>}
     {limitRadius=17, drivingBone {boneName=spine_04_in}, sphereLocalOffset=<0 0 0>}
     {limitRadius=17, drivingBone {boneName=spine_03_in}, sphereLocalOffset=<0 0 0>}
     {limitRadius=17, drivingBone {boneName=spine_02_in}, sphereLocalOffset=<0 0 0>}
     {limitRadius=7, drivingBone {boneName=upperarm_l_in}, sphereLocalOffset=<15 0 0>}
    ]
   }
   n2 = Anim Dynamics - Bone: dyn_chain_02!(ComponentPose=@n1, Alpha=1, GravityOverride:Vector="0,0,-980") -> (Pose) {
    angularDampingOverride=0.9
    angularSpringConstant=5000
    bOverrideAngularDamping=true
    bUseGravityOverride=true
    bUseSphericalLimits=true
    linearDampingOverride=1
    linearSpringConstant=500
    simulationSpace=RootRelative
    boundBone {boneName=dyn_chain_02}
    componentAppliedLinearAccClamp=<310 310 310>
    componentLinearAccScale=<0.9 0.9 0.9>
    componentLinearVelScale=<0.9 0.9 0.9>
    physicsBodyDefinitions [{
      collisionType=CustomSphere
      sphereCollisionRadius=1.5
      boundBone {boneName=dyn_chain_02}
      boxExtents=<36 26 26>
      constraintSetup {angularLimitsMax=<0 0 15>, angularLimitsMin=<0 -15 -15>, angularTarget=<1 0 0>}
      localJointOffset=<5 0 0>
     }]
    sphericalLimits [
     {limitRadius=10, drivingBone {boneName=upperarm_l_in}, sphereLocalOffset=<0 0 0>}
     {limitRadius=17, drivingBone {boneName=spine_05_in}, sphereLocalOffset=<0 0 0>}
     {limitRadius=17, drivingBone {boneName=spine_04_in}, sphereLocalOffset=<0 0 0>}
     {limitRadius=17, drivingBone {boneName=spine_03_in}, sphereLocalOffset=<0 0 0>}
     {limitRadius=17, drivingBone {boneName=spine_02_in}, sphereLocalOffset=<0 0 0>}
     {limitRadius=7, drivingBone {boneName=upperarm_l_in}, sphereLocalOffset=<15 0 0>}
    ]
   }
   n3 = Anim Dynamics - Bone: dyn_chain_03!(ComponentPose=@n2, Alpha=1, GravityOverride:Vector="0,0,-980") -> (Pose) {
    angularDampingOverride=0.9
    angularSpringConstant=5000
    bOverrideAngularDamping=true
    bUseGravityOverride=true
    bUseSphericalLimits=true
    linearDampingOverride=1
    linearSpringConstant=500
    simulationSpace=RootRelative
    boundBone {boneName=dyn_chain_03}
    componentAppliedLinearAccClamp=<315 315 315>
    componentLinearAccScale=<0.9 0.9 0.9>
    componentLinearVelScale=<0.9 0.9 0.9>
    physicsBodyDefinitions [{
      collisionType=CustomSphere
      sphereCollisionRadius=1.5
      boundBone {boneName=dyn_chain_03}
      boxExtents=<35 25 25>
      constraintSetup {angularLimitsMax=<0 0 15>, angularLimitsMin=<0 -15 -15>, angularTarget=<1 0 0>}
      localJointOffset=<5 0 0>
     }]
    sphericalLimits [
     {limitRadius=10, drivingBone {boneName=upperarm_l_in}, sphereLocalOffset=<0 0 0>}
     {limitRadius=17, drivingBone {boneName=spine_05_in}, sphereLocalOffset=<0 0 0>}
     {limitRadius=17, drivingBone {boneName=spine_04_in}, sphereLocalOffset=<0 0 0>}
     {limitRadius=17, drivingBone {boneName=spine_03_in}, sphereLocalOffset=<0 0 0>}
     {limitRadius=17, drivingBone {boneName=spine_02_in}, sphereLocalOffset=<0 0 0>}
     {limitRadius=7, drivingBone {boneName=upperarm_l_in}, sphereLocalOffset=<15 0 0>}
    ]
   }
   n4 = Anim Dynamics - Bone: dyn_chain_04!(ComponentPose=@n3, Alpha=1, GravityOverride:Vector="0,0,-980") -> (Pose) {
    angularDampingOverride=0.9
    angularSpringConstant=5000
    bOverrideAngularDamping=true
    bUseGravityOverride=true
    bUseSphericalLimits=true
    linearDampingOverride=1
    linearSpringConstant=500
    simulationSpace=RootRelative
    boundBone {boneName=dyn_chain_04}
    componentAppliedLinearAccClamp=<320 320 320>
    componentLinearAccScale=<0.9 0.9 0.9>
    componentLinearVelScale=<0.9 0.9 0.9>
    physicsBodyDefinitions [{
      collisionType=CustomSphere
      sphereCollisionRadius=1.5
      boundBone {boneName=dyn_chain_04}
      boxExtents=<34 24 24>
      constraintSetup {angularLimitsMax=<0 0 15>, angularLimitsMin=<0 -15 -15>, angularTarget=<1 0 0>}
      localJointOffset=<5 0 0>
     }]
    sphericalLimits [
     {limitRadius=10, drivingBone {boneName=upperarm_l_in}, sphereLocalOffset=<0 0 0>}
     {limitRadius=17, drivingBone {boneName=spine_05_in}, sphereLocalOffset=<0 0 0>}
     {limitRadius=17, drivingBone {boneName=spine_04_in}, sphereLocalOffset=<0 0 0>}
     {limitRadius=17, drivingBone {boneName=spine_03_in}, sphereLocalOffset=<0 0 0>}
     {limitRadius=17, drivingBone {boneName=spine_02_in}, sphereLocalOffset=<0 0 0>}
     {limitRadius=7, drivingBone {boneName=upperarm_l_in}, sphereLocalOffset=<15 0 0>}
    ]
   }
   n5 = Anim Dynamics - Bone: dyn_chain_05!(ComponentPose=@n4, Alpha=1, GravityOverride:Vector="0,0,-980") -> (Pose) {
    angularDampingOverride=0.9
    angularSpringConstant=5000
    bOverrideAngularDamping=true
    bUseGravityOverride=true
    bUseSphericalLimits=true
    linearDampingOverride=1
    linearSpringConstant=500
    simulationSpace=RootRelative
    boundBone {boneName=dyn_chain_05}
    componentAppliedLinearAccClamp=<340 340 340>
    componentLinearAccScale=<0.9 0.9 0.9>
    componentLinearVelScale=<0.9 0.9 0.9>
    physicsBodyDefinitions [{
      collisionType=CustomSphere
      sphereCollisionRadius=1.5
      boundBone {boneName=dyn_chain_05}
      boxExtents=<33 23 23>
      constraintSetup {angularLimitsMax=<0 0 15>, angularLimitsMin=<0 -15 -15>, angularTarget=<1 0 0>}
      localJointOffset=<5 0 0>
     }]
    sphericalLimits [
     {limitRadius=10, drivingBone {boneName=upperarm_l_in}, sphereLocalOffset=<0 0 0>}
     {limitRadius=17, drivingBone {boneName=spine_05_in}, sphereLocalOffset=<0 0 0>}
     {limitRadius=17, drivingBone {boneName=spine_04_in}, sphereLocalOffset=<0 0 0>}
     {limitRadius=17, drivingBone {boneName=spine_03_in}, sphereLocalOffset=<0 0 0>}
     {limitRadius=17, drivingBone {boneName=spine_02_in}, sphereLocalOffset=<0 0 0>}
     {limitRadius=7, drivingBone {boneName=upperarm_l_in}, sphereLocalOffset=<15 0 0>}
    ]
   }
   n6 = Anim Dynamics - Bone: dyn_chain_06!(ComponentPose=@n5, Alpha=1, GravityOverride:Vector="0,0,-980") -> (Pose) {
    angularDampingOverride=0.9
    angularSpringConstant=5000
    bOverrideAngularDamping=true
    bUseGravityOverride=true
    bUseSphericalLimits=true
    linearDampingOverride=1
    linearSpringConstant=500
    simulationSpace=RootRelative
    boundBone {boneName=dyn_chain_06}
    componentAppliedLinearAccClamp=<360 360 360>
    componentLinearAccScale=<0.9 0.9 0.9>
    componentLinearVelScale=<0.9 0.9 0.9>
    physicsBodyDefinitions [{
      collisionType=CustomSphere
      sphereCollisionRadius=1.5
      boundBone {boneName=dyn_chain_06}
      boxExtents=<32 22 22>
      constraintSetup {angularLimitsMax=<0 0 15>, angularLimitsMin=<0 -15 -15>, angularTarget=<1 0 0>}
      localJointOffset=<5 0 0>
     }]
    sphericalLimits [
     {limitRadius=10, drivingBone {boneName=upperarm_l_in}, sphereLocalOffset=<0 0 0>}
     {limitRadius=17, drivingBone {boneName=spine_05_in}, sphereLocalOffset=<0 0 0>}
     {limitRadius=17, drivingBone {boneName=spine_04_in}, sphereLocalOffset=<0 0 0>}
     {limitRadius=17, drivingBone {boneName=spine_03_in}, sphereLocalOffset=<0 0 0>}
     {limitRadius=17, drivingBone {boneName=spine_02_in}, sphereLocalOffset=<0 0 0>}
     {limitRadius=7, drivingBone {boneName=upperarm_l_in}, sphereLocalOffset=<15 0 0>}
    ]
   }
   n7 = Anim Dynamics - Bone: dyn_chain_07!(ComponentPose=@n6, Alpha=1, GravityOverride:Vector="0,0,-980") -> (Pose) {
    angularDampingOverride=0.9
    angularSpringConstant=5000
    bOverrideAngularDamping=true
    bUseGravityOverride=true
    bUseSphericalLimits=true
    linearDampingOverride=1
    linearSpringConstant=500
    simulationSpace=RootRelative
    boundBone {boneName=dyn_chain_07}
    componentAppliedLinearAccClamp=<380 380 380>
    componentLinearAccScale=<0.9 0.9 0.9>
    componentLinearVelScale=<0.9 0.9 0.9>
    physicsBodyDefinitions [{
      collisionType=CustomSphere
      sphereCollisionRadius=1.5
      boundBone {boneName=dyn_chain_07}
      boxExtents=<31 21 21>
      constraintSetup {angularLimitsMax=<0 0 15>, angularLimitsMin=<0 -15 -15>, angularTarget=<1 0 0>}
      localJointOffset=<5 0 0>
     }]
    sphericalLimits [
     {limitRadius=10, drivingBone {boneName=upperarm_l_in}, sphereLocalOffset=<0 0 0>}
     {limitRadius=17, drivingBone {boneName=spine_05_in}, sphereLocalOffset=<0 0 0>}
     {limitRadius=17, drivingBone {boneName=spine_04_in}, sphereLocalOffset=<0 0 0>}
     {limitRadius=17, drivingBone {boneName=spine_03_in}, sphereLocalOffset=<0 0 0>}
     {limitRadius=17, drivingBone {boneName=spine_02_in}, sphereLocalOffset=<0 0 0>}
     {limitRadius=7, drivingBone {boneName=upperarm_l_in}, sphereLocalOffset=<15 0 0>}
    ]
   }
   n8 = Anim Dynamics - Bone: dyn_chain_08!(ComponentPose=@n7, Alpha=1, GravityOverride:Vector="0,0,-980") -> (Pose) {
    angularDampingOverride=0.9
    angularSpringConstant=5000
    bOverrideAngularDamping=true
    bUseGravityOverride=true
    bUseSphericalLimits=true
    linearDampingOverride=1
    linearSpringConstant=500
    simulationSpace=RootRelative
    boundBone {boneName=dyn_chain_08}
    componentAppliedLinearAccClamp=<400 400 400>
    componentLinearAccScale=<0.9 0.9 0.9>
    componentLinearVelScale=<0.9 0.9 0.9>
    physicsBodyDefinitions [{
      collisionType=CustomSphere
      sphereCollisionRadius=1.5
      boundBone {boneName=dyn_chain_08}
      boxExtents=<30 20 20>
      constraintSetup {angularLimitsMax=<0 0 15>, angularLimitsMin=<0 -15 -15>, angularTarget=<1 0 0>}
      localJointOffset=<5 0 0>
     }]
    sphericalLimits [
     {limitRadius=10, drivingBone {boneName=upperarm_l_in}, sphereLocalOffset=<0 0 0>}
     {limitRadius=17, drivingBone {boneName=spine_05_in}, sphereLocalOffset=<0 0 0>}
     {limitRadius=17, drivingBone {boneName=spine_04_in}, sphereLocalOffset=<0 0 0>}
     {limitRadius=17, drivingBone {boneName=spine_03_in}, sphereLocalOffset=<0 0 0>}
     {limitRadius=17, drivingBone {boneName=spine_02_in}, sphereLocalOffset=<0 0 0>}
     {limitRadius=7, drivingBone {boneName=upperarm_l_in}, sphereLocalOffset=<15 0 0>}
    ]
   }
   n9 = Anim Dynamics - Bone: dyn_chain_09!(ComponentPose=@n8, Alpha=1, GravityOverride:Vector="0,0,-980") -> (Pose) {
    angularDampingOverride=0.9
    angularSpringConstant=5000
    bOverrideAngularDamping=true
    bUseGravityOverride=true
    bUseSphericalLimits=true
    linearDampingOverride=1
    linearSpringConstant=500
    simulationSpace=RootRelative
    boundBone {boneName=dyn_chain_09}
    componentAppliedLinearAccClamp=<420 420 420>
    componentLinearAccScale=<0.9 0.9 0.9>
    componentLinearVelScale=<0.9 0.9 0.9>
    physicsBodyDefinitions [{
      collisionType=CustomSphere
      sphereCollisionRadius=1.5
      boundBone {boneName=dyn_chain_09}
      boxExtents=<29 19 19>
      constraintSetup {angularLimitsMax=<0 0 15>, angularLimitsMin=<0 -15 -15>, angularTarget=<1 0 0>}
      localJointOffset=<5 0 0>
     }]
    sphericalLimits [
     {limitRadius=10, drivingBone {boneName=upperarm_l_in}, sphereLocalOffset=<0 0 0>}
     {limitRadius=17, drivingBone {boneName=spine_05_in}, sphereLocalOffset=<0 0 0>}
     {limitRadius=17, drivingBone {boneName=spine_04_in}, sphereLocalOffset=<0 0 0>}
     {limitRadius=17, drivingBone {boneName=spine_03_in}, sphereLocalOffset=<0 0 0>}
     {limitRadius=17, drivingBone {boneName=spine_02_in}, sphereLocalOffset=<0 0 0>}
     {limitRadius=7, drivingBone {boneName=upperarm_l_in}, sphereLocalOffset=<15 0 0>}
    ]
   }
   n10 = Anim Dynamics - Bone: dyn_handle!(ComponentPose=@n9, Alpha=1, GravityOverride:Vector="0,0,-980") -> (Pose) {
    angularDampingOverride=0.9
    angularSpringConstant=5000
    bOverrideAngularDamping=true
    bUseGravityOverride=true
    bUseSphericalLimits=true
    linearDampingOverride=1
    linearSpringConstant=500
    simulationSpace=RootRelative
    boundBone {boneName=dyn_handle}
    componentAppliedLinearAccClamp=<440 440 440>
    componentLinearAccScale=<0.9 0.9 0.9>
    componentLinearVelScale=<0.9 0.9 0.9>
    physicsBodyDefinitions [{
      collisionType=CustomSphere
      sphereCollisionRadius=1.5
      boundBone {boneName=dyn_handle}
      boxExtents=<28 18 18>
      constraintSetup {angularLimitsMax=<0 0 15>, angularLimitsMin=<0 -15 -15>, angularTarget=<1 0 0>}
      localJointOffset=<5 0 0>
     }]
    sphericalLimits [
     {limitRadius=11, drivingBone {boneName=upperarm_l_in}, sphereLocalOffset=<0 0 0>}
     {limitRadius=17, drivingBone {boneName=spine_05_in}, sphereLocalOffset=<0 0 0>}
     {limitRadius=17, drivingBone {boneName=spine_04_in}, sphereLocalOffset=<0 0 0>}
     {limitRadius=17, drivingBone {boneName=spine_03_in}, sphereLocalOffset=<0 0 0>}
     {limitRadius=17, drivingBone {boneName=spine_02_in}, sphereLocalOffset=<0 0 0>}
     {limitRadius=8, drivingBone {boneName=upperarm_l_in}, sphereLocalOffset=<15 0 0>}
    ]
   }
   n11 = Anim Dynamics - Bone: dyn_chain_01(ComponentPose=Local To Component(LocalPose=Use cached pose 'start pose'()), Alpha=1, GravityOverride="0,0,-980") {
    angularDampingOverride=0.9
    angularSpringConstant=5000
    bOverrideAngularDamping=true
    bUseGravityOverride=true
    bUseSphericalLimits=true
    linearDampingOverride=1
    linearSpringConstant=500
    simulationSpace=RootRelative
    boundBone {boneName=dyn_chain_01}
    componentAppliedLinearAccClamp=<305 305 305>
    componentLinearAccScale=<0.9 0.9 0.9>
    componentLinearVelScale=<0.9 0.9 0.9>
    physicsBodyDefinitions [{
      collisionType=CustomSphere
      sphereCollisionRadius=1.5
      boundBone {boneName=dyn_chain_01}
      boxExtents=<37 27 27>
      constraintSetup {angularLimitsMax=<0 0 15>, angularLimitsMin=<0 -15 -15>, angularTarget=<1 0 0>}
      localJointOffset=<5 0 0>
     }]
    sphericalLimits [
     {limitRadius=11, drivingBone {boneName=upperarm_l_in}, sphereLocalOffset=<0 0 0>}
     {limitRadius=20, drivingBone {boneName=spine_05_in}, sphereLocalOffset=<0 0 0>}
     {limitRadius=20, drivingBone {boneName=spine_04_in}, sphereLocalOffset=<0 0 0>}
     {limitRadius=22, drivingBone {boneName=spine_03_in}, sphereLocalOffset=<0 0 0>}
     {limitRadius=22, drivingBone {boneName=spine_02_in}, sphereLocalOffset=<0 0 0>}
     {limitRadius=9, drivingBone {boneName=upperarm_l_in}, sphereLocalOffset=<15 0 0>}
    ]
   }
   n12 = Anim Dynamics - Bone: dyn_chain_02(ComponentPose=@n11, Alpha=1, GravityOverride="0,0,-980") {
    angularDampingOverride=0.9
    angularSpringConstant=5000
    bOverrideAngularDamping=true
    bUseGravityOverride=true
    bUseSphericalLimits=true
    linearDampingOverride=1
    linearSpringConstant=500
    simulationSpace=RootRelative
    boundBone {boneName=dyn_chain_02}
    componentAppliedLinearAccClamp=<310 310 310>
    componentLinearAccScale=<0.9 0.9 0.9>
    componentLinearVelScale=<0.9 0.9 0.9>
    physicsBodyDefinitions [{
      collisionType=CustomSphere
      sphereCollisionRadius=1.5
      boundBone {boneName=dyn_chain_02}
      boxExtents=<36 26 26>
      constraintSetup {angularLimitsMax=<0 0 15>, angularLimitsMin=<0 -15 -15>, angularTarget=<1 0 0>}
      localJointOffset=<5 0 0>
     }]
    sphericalLimits [
     {limitRadius=11, drivingBone {boneName=upperarm_l_in}, sphereLocalOffset=<0 0 0>}
     {limitRadius=20, drivingBone {boneName=spine_05_in}, sphereLocalOffset=<0 0 0>}
     {limitRadius=20, drivingBone {boneName=spine_04_in}, sphereLocalOffset=<0 0 0>}
     {limitRadius=22, drivingBone {boneName=spine_03_in}, sphereLocalOffset=<0 0 0>}
     {limitRadius=22, drivingBone {boneName=spine_02_in}, sphereLocalOffset=<0 0 0>}
     {limitRadius=9, drivingBone {boneName=upperarm_l_in}, sphereLocalOffset=<15 0 0>}
    ]
   }
   n13 = Anim Dynamics - Bone: dyn_chain_03(ComponentPose=@n12, Alpha=1, GravityOverride="0,0,-980") {
    angularDampingOverride=0.9
    angularSpringConstant=5000
    bOverrideAngularDamping=true
    bUseGravityOverride=true
    bUseSphericalLimits=true
    linearDampingOverride=1
    linearSpringConstant=500
    simulationSpace=RootRelative
    boundBone {boneName=dyn_chain_03}
    componentAppliedLinearAccClamp=<315 315 315>
    componentLinearAccScale=<0.9 0.9 0.9>
    componentLinearVelScale=<0.9 0.9 0.9>
    physicsBodyDefinitions [{
      collisionType=CustomSphere
      sphereCollisionRadius=1.5
      boundBone {boneName=dyn_chain_03}
      boxExtents=<35 25 25>
      constraintSetup {angularLimitsMax=<0 0 15>, angularLimitsMin=<0 -15 -15>, angularTarget=<1 0 0>}
      localJointOffset=<5 0 0>
     }]
    sphericalLimits [
     {limitRadius=11, drivingBone {boneName=upperarm_l_in}, sphereLocalOffset=<0 0 0>}
     {limitRadius=20, drivingBone {boneName=spine_05_in}, sphereLocalOffset=<0 0 0>}
     {limitRadius=20, drivingBone {boneName=spine_04_in}, sphereLocalOffset=<0 0 0>}
     {limitRadius=22, drivingBone {boneName=spine_03_in}, sphereLocalOffset=<0 0 0>}
     {limitRadius=22, drivingBone {boneName=spine_02_in}, sphereLocalOffset=<0 0 0>}
     {limitRadius=9, drivingBone {boneName=upperarm_l_in}, sphereLocalOffset=<15 0 0>}
    ]
   }
   n14 = Anim Dynamics - Bone: dyn_chain_04(ComponentPose=@n13, Alpha=1, GravityOverride="0,0,-980") {
    angularDampingOverride=0.9
    angularSpringConstant=5000
    bOverrideAngularDamping=true
    bUseGravityOverride=true
    bUseSphericalLimits=true
    linearDampingOverride=1
    linearSpringConstant=500
    simulationSpace=RootRelative
    boundBone {boneName=dyn_chain_04}
    componentAppliedLinearAccClamp=<320 320 320>
    componentLinearAccScale=<0.9 0.9 0.9>
    componentLinearVelScale=<0.9 0.9 0.9>
    physicsBodyDefinitions [{
      collisionType=CustomSphere
      sphereCollisionRadius=1.5
      boundBone {boneName=dyn_chain_04}
      boxExtents=<34 24 24>
      constraintSetup {angularLimitsMax=<0 0 15>, angularLimitsMin=<0 -15 -15>, angularTarget=<1 0 0>}
      localJointOffset=<5 0 0>
     }]
    sphericalLimits [
     {limitRadius=11, drivingBone {boneName=upperarm_l_in}, sphereLocalOffset=<0 0 0>}
     {limitRadius=20, drivingBone {boneName=spine_05_in}, sphereLocalOffset=<0 0 0>}
     {limitRadius=20, drivingBone {boneName=spine_04_in}, sphereLocalOffset=<0 0 0>}
     {limitRadius=22, drivingBone {boneName=spine_03_in}, sphereLocalOffset=<0 0 0>}
     {limitRadius=22, drivingBone {boneName=spine_02_in}, sphereLocalOffset=<0 0 0>}
     {limitRadius=9, drivingBone {boneName=upperarm_l_in}, sphereLocalOffset=<15 0 0>}
    ]
   }
   n15 = Anim Dynamics - Bone: dyn_chain_05(ComponentPose=@n14, Alpha=1, GravityOverride="0,0,-980") {
    angularDampingOverride=0.9
    angularSpringConstant=5000
    bOverrideAngularDamping=true
    bUseGravityOverride=true
    bUseSphericalLimits=true
    linearDampingOverride=1
    linearSpringConstant=500
    simulationSpace=RootRelative
    boundBone {boneName=dyn_chain_05}
    componentAppliedLinearAccClamp=<340 340 340>
    componentLinearAccScale=<0.9 0.9 0.9>
    componentLinearVelScale=<0.9 0.9 0.9>
    physicsBodyDefinitions [{
      collisionType=CustomSphere
      sphereCollisionRadius=1.5
      boundBone {boneName=dyn_chain_05}
      boxExtents=<33 23 23>
      constraintSetup {angularLimitsMax=<0 0 15>, angularLimitsMin=<0 -15 -15>, angularTarget=<1 0 0>}
      localJointOffset=<5 0 0>
     }]
    sphericalLimits [
     {limitRadius=11, drivingBone {boneName=upperarm_l_in}, sphereLocalOffset=<0 0 0>}
     {limitRadius=20, drivingBone {boneName=spine_05_in}, sphereLocalOffset=<0 0 0>}
     {limitRadius=20, drivingBone {boneName=spine_04_in}, sphereLocalOffset=<0 0 0>}
     {limitRadius=22, drivingBone {boneName=spine_03_in}, sphereLocalOffset=<0 0 0>}
     {limitRadius=22, drivingBone {boneName=spine_02_in}, sphereLocalOffset=<0 0 0>}
     {limitRadius=9, drivingBone {boneName=upperarm_l_in}, sphereLocalOffset=<15 0 0>}
    ]
   }
   n16 = Anim Dynamics - Bone: dyn_chain_06(ComponentPose=@n15, Alpha=1, GravityOverride="0,0,-980") {
    angularDampingOverride=0.9
    angularSpringConstant=5000
    bOverrideAngularDamping=true
    bUseGravityOverride=true
    bUseSphericalLimits=true
    linearDampingOverride=1
    linearSpringConstant=500
    simulationSpace=RootRelative
    boundBone {boneName=dyn_chain_06}
    componentAppliedLinearAccClamp=<360 360 360>
    componentLinearAccScale=<0.9 0.9 0.9>
    componentLinearVelScale=<0.9 0.9 0.9>
    physicsBodyDefinitions [{
      collisionType=CustomSphere
      sphereCollisionRadius=1.5
      boundBone {boneName=dyn_chain_06}
      boxExtents=<32 22 22>
      constraintSetup {angularLimitsMax=<0 0 15>, angularLimitsMin=<0 -15 -15>, angularTarget=<1 0 0>}
      localJointOffset=<5 0 0>
     }]
    sphericalLimits [
     {limitRadius=11, drivingBone {boneName=upperarm_l_in}, sphereLocalOffset=<0 0 0>}
     {limitRadius=20, drivingBone {boneName=spine_05_in}, sphereLocalOffset=<0 0 0>}
     {limitRadius=20, drivingBone {boneName=spine_04_in}, sphereLocalOffset=<0 0 0>}
     {limitRadius=22, drivingBone {boneName=spine_03_in}, sphereLocalOffset=<0 0 0>}
     {limitRadius=22, drivingBone {boneName=spine_02_in}, sphereLocalOffset=<0 0 0>}
     {limitRadius=9, drivingBone {boneName=upperarm_l_in}, sphereLocalOffset=<15 0 0>}
    ]
   }
   n17 = Anim Dynamics - Bone: dyn_chain_07(ComponentPose=@n16, Alpha=1, GravityOverride="0,0,-980") {
    angularDampingOverride=0.9
    angularSpringConstant=5000
    bOverrideAngularDamping=true
    bUseGravityOverride=true
    bUseSphericalLimits=true
    linearDampingOverride=1
    linearSpringConstant=500
    simulationSpace=RootRelative
    boundBone {boneName=dyn_chain_07}
    componentAppliedLinearAccClamp=<380 380 380>
    componentLinearAccScale=<0.9 0.9 0.9>
    componentLinearVelScale=<0.9 0.9 0.9>
    physicsBodyDefinitions [{
      collisionType=CustomSphere
      sphereCollisionRadius=1.5
      boundBone {boneName=dyn_chain_07}
      boxExtents=<31 21 21>
      constraintSetup {angularLimitsMax=<0 0 15>, angularLimitsMin=<0 -15 -15>, angularTarget=<1 0 0>}
      localJointOffset=<5 0 0>
     }]
    sphericalLimits [
     {limitRadius=11, drivingBone {boneName=upperarm_l_in}, sphereLocalOffset=<0 0 0>}
     {limitRadius=20, drivingBone {boneName=spine_05_in}, sphereLocalOffset=<0 0 0>}
     {limitRadius=20, drivingBone {boneName=spine_04_in}, sphereLocalOffset=<0 0 0>}
     {limitRadius=22, drivingBone {boneName=spine_03_in}, sphereLocalOffset=<0 0 0>}
     {limitRadius=22, drivingBone {boneName=spine_02_in}, sphereLocalOffset=<0 0 0>}
     {limitRadius=9, drivingBone {boneName=upperarm_l_in}, sphereLocalOffset=<15 0 0>}
    ]
   }
   n18 = Anim Dynamics - Bone: dyn_chain_08(ComponentPose=@n17, Alpha=1, GravityOverride="0,0,-980") {
    angularDampingOverride=0.9
    angularSpringConstant=5000
    bOverrideAngularDamping=true
    bUseGravityOverride=true
    bUseSphericalLimits=true
    linearDampingOverride=1
    linearSpringConstant=500
    simulationSpace=RootRelative
    boundBone {boneName=dyn_chain_08}
    componentAppliedLinearAccClamp=<400 400 400>
    componentLinearAccScale=<0.9 0.9 0.9>
    componentLinearVelScale=<0.9 0.9 0.9>
    physicsBodyDefinitions [{
      collisionType=CustomSphere
      sphereCollisionRadius=1.5
      boundBone {boneName=dyn_chain_08}
      boxExtents=<30 20 20>
      constraintSetup {angularLimitsMax=<0 0 15>, angularLimitsMin=<0 -15 -15>, angularTarget=<1 0 0>}
      localJointOffset=<5 0 0>
     }]
    sphericalLimits [
     {limitRadius=11, drivingBone {boneName=upperarm_l_in}, sphereLocalOffset=<0 0 0>}
     {limitRadius=20, drivingBone {boneName=spine_05_in}, sphereLocalOffset=<0 0 0>}
     {limitRadius=20, drivingBone {boneName=spine_04_in}, sphereLocalOffset=<0 0 0>}
     {limitRadius=22, drivingBone {boneName=spine_03_in}, sphereLocalOffset=<0 0 0>}
     {limitRadius=22, drivingBone {boneName=spine_02_in}, sphereLocalOffset=<0 0 0>}
     {limitRadius=9, drivingBone {boneName=upperarm_l_in}, sphereLocalOffset=<15 0 0>}
    ]
   }
   n19 = Anim Dynamics - Bone: dyn_chain_09(ComponentPose=@n18, Alpha=1, GravityOverride="0,0,-980") {
    angularDampingOverride=0.9
    angularSpringConstant=5000
    bOverrideAngularDamping=true
    bUseGravityOverride=true
    bUseSphericalLimits=true
    linearDampingOverride=1
    linearSpringConstant=500
    simulationSpace=RootRelative
    boundBone {boneName=dyn_chain_09}
    componentAppliedLinearAccClamp=<420 420 420>
    componentLinearAccScale=<0.9 0.9 0.9>
    componentLinearVelScale=<0.9 0.9 0.9>
    physicsBodyDefinitions [{
      collisionType=CustomSphere
      sphereCollisionRadius=1.5
      boundBone {boneName=dyn_chain_09}
      boxExtents=<29 19 19>
      constraintSetup {angularLimitsMax=<0 0 15>, angularLimitsMin=<0 -15 -15>, angularTarget=<1 0 0>}
      localJointOffset=<5 0 0>
     }]
    sphericalLimits [
     {limitRadius=11, drivingBone {boneName=upperarm_l_in}, sphereLocalOffset=<0 0 0>}
     {limitRadius=20, drivingBone {boneName=spine_05_in}, sphereLocalOffset=<0 0 0>}
     {limitRadius=20, drivingBone {boneName=spine_04_in}, sphereLocalOffset=<0 0 0>}
     {limitRadius=22, drivingBone {boneName=spine_03_in}, sphereLocalOffset=<0 0 0>}
     {limitRadius=22, drivingBone {boneName=spine_02_in}, sphereLocalOffset=<0 0 0>}
     {limitRadius=9, drivingBone {boneName=upperarm_l_in}, sphereLocalOffset=<15 0 0>}
    ]
   }
   n20 = Anim Dynamics - Bone: dyn_handle(ComponentPose=@n19, Alpha=1, GravityOverride="0,0,-980") {
    angularDampingOverride=0.9
    angularSpringConstant=5000
    bOverrideAngularDamping=true
    bUseGravityOverride=true
    bUseSphericalLimits=true
    linearDampingOverride=1
    linearSpringConstant=500
    simulationSpace=RootRelative
    boundBone {boneName=dyn_handle}
    componentAppliedLinearAccClamp=<440 440 440>
    componentLinearAccScale=<0.9 0.9 0.9>
    componentLinearVelScale=<0.9 0.9 0.9>
    physicsBodyDefinitions [{
      collisionType=CustomSphere
      sphereCollisionRadius=1.5
      boundBone {boneName=dyn_handle}
      boxExtents=<28 18 18>
      constraintSetup {angularLimitsMax=<0 0 15>, angularLimitsMin=<0 -15 -15>, angularTarget=<1 0 0>}
      localJointOffset=<5 0 0>
     }]
    sphericalLimits [
     {limitRadius=11, drivingBone {boneName=upperarm_l_in}, sphereLocalOffset=<0 0 0>}
     {limitRadius=20, drivingBone {boneName=spine_05_in}, sphereLocalOffset=<0 0 0>}
     {limitRadius=20, drivingBone {boneName=spine_04_in}, sphereLocalOffset=<0 0 0>}
     {limitRadius=22, drivingBone {boneName=spine_03_in}, sphereLocalOffset=<0 0 0>}
     {limitRadius=22, drivingBone {boneName=spine_02_in}, sphereLocalOffset=<0 0 0>}
     {limitRadius=9, drivingBone {boneName=upperarm_l_in}, sphereLocalOffset=<15 0 0>}
    ]
   }
   n21 = Blend Poses by bool(
    BlendPose_0=Component To Local(ComponentPose=@n20)
    BlendPose_1=Component To Local(ComponentPose=@n10)
    bActiveValue=Get isMale() -> (isMale:bool)
   )
   n22 = Anim Dynamics - Bone: dyn_chain_01(ComponentPose=Local To Component(LocalPose=Use cached pose 'start pose'()), Alpha=1, GravityOverride="0,0,-980") {
    angularDampingOverride=0.9
    angularSpringConstant=5000
    bOverrideAngularDamping=true
    bUseGravityOverride=true
    linearDampingOverride=1
    linearSpringConstant=500
    simulationSpace=RootRelative
    boundBone {boneName=dyn_chain_01}
    componentAppliedLinearAccClamp=<100 100 100>
    componentLinearAccScale=<0.1 0.1 0.1>
    componentLinearVelScale=<0.1 0.1 0.1>
    physicsBodyDefinitions [{
      collisionType=CustomSphere
      sphereCollisionRadius=2
      boundBone {boneName=dyn_chain_01}
      boxExtents=<37 27 27>
      constraintSetup {angularLimitsMax=<0 0 2.5>, angularLimitsMin=<0 -2.5 -5>, angularTarget=<1 0 0>}
      localJointOffset=<5 0 0>
     }]
    sphericalLimits [
     {limitRadius=10, drivingBone {boneName=upperarm_l_in}, sphereLocalOffset=<0 0 0>}
     {limitRadius=17, drivingBone {boneName=spine_05_in}, sphereLocalOffset=<0 0 0>}
     {limitRadius=17, drivingBone {boneName=spine_04_in}, sphereLocalOffset=<0 0 0>}
     {limitRadius=17, drivingBone {boneName=spine_03_in}, sphereLocalOffset=<0 0 0>}
     {limitRadius=17, drivingBone {boneName=spine_02_in}, sphereLocalOffset=<0 0 0>}
     {limitRadius=8, drivingBone {boneName=upperarm_l_in}, sphereLocalOffset=<15 0 0>}
    ]
   }
   n23 = Anim Dynamics - Bone: dyn_chain_02(ComponentPose=@n22, Alpha=1, GravityOverride="0,0,-980") {
    angularDampingOverride=0.9
    angularSpringConstant=5000
    bOverrideAngularDamping=true
    bUseGravityOverride=true
    linearDampingOverride=1
    linearSpringConstant=500
    simulationSpace=RootRelative
    boundBone {boneName=dyn_chain_02}
    componentAppliedLinearAccClamp=<100 100 100>
    componentLinearAccScale=<0.1 0.1 0.1>
    componentLinearVelScale=<0.1 0.1 0.1>
    physicsBodyDefinitions [{
      collisionType=CustomSphere
      sphereCollisionRadius=2
      boundBone {boneName=dyn_chain_02}
      boxExtents=<36 26 26>
      constraintSetup {angularLimitsMax=<0 0 2.5>, angularLimitsMin=<0 -2.5 -5>, angularTarget=<1 0 0>}
      localJointOffset=<5 0 0>
     }]
    sphericalLimits [
     {limitRadius=10, drivingBone {boneName=upperarm_l_in}, sphereLocalOffset=<0 0 0>}
     {limitRadius=17, drivingBone {boneName=spine_05_in}, sphereLocalOffset=<0 0 0>}
     {limitRadius=17, drivingBone {boneName=spine_04_in}, sphereLocalOffset=<0 0 0>}
     {limitRadius=17, drivingBone {boneName=spine_03_in}, sphereLocalOffset=<0 0 0>}
     {limitRadius=17, drivingBone {boneName=spine_02_in}, sphereLocalOffset=<0 0 0>}
     {limitRadius=8, drivingBone {boneName=upperarm_l_in}, sphereLocalOffset=<15 0 0>}
    ]
   }
   n24 = Anim Dynamics - Bone: dyn_chain_03(ComponentPose=@n23, Alpha=1, GravityOverride="0,0,-980") {
    angularDampingOverride=0.9
    angularSpringConstant=5000
    bOverrideAngularDamping=true
    bUseGravityOverride=true
    linearDampingOverride=1
    linearSpringConstant=500
    simulationSpace=RootRelative
    boundBone {boneName=dyn_chain_03}
    componentAppliedLinearAccClamp=<100 100 100>
    componentLinearAccScale=<0.1 0.1 0.1>
    componentLinearVelScale=<0.1 0.1 0.1>
    physicsBodyDefinitions [{
      collisionType=CustomSphere
      sphereCollisionRadius=2
      boundBone {boneName=dyn_chain_03}
      boxExtents=<35 25 25>
      constraintSetup {angularLimitsMax=<0 0 2.5>, angularLimitsMin=<0 -2.5 -5>, angularTarget=<1 0 0>}
      localJointOffset=<5 0 0>
     }]
    sphericalLimits [
     {limitRadius=10, drivingBone {boneName=upperarm_l_in}, sphereLocalOffset=<0 0 0>}
     {limitRadius=17, drivingBone {boneName=spine_05_in}, sphereLocalOffset=<0 0 0>}
     {limitRadius=17, drivingBone {boneName=spine_04_in}, sphereLocalOffset=<0 0 0>}
     {limitRadius=17, drivingBone {boneName=spine_03_in}, sphereLocalOffset=<0 0 0>}
     {limitRadius=17, drivingBone {boneName=spine_02_in}, sphereLocalOffset=<0 0 0>}
     {limitRadius=8, drivingBone {boneName=upperarm_l_in}, sphereLocalOffset=<15 0 0>}
    ]
   }
   n25 = Anim Dynamics - Bone: dyn_chain_04(ComponentPose=@n24, Alpha=1, GravityOverride="0,0,-980") {
    angularDampingOverride=0.9
    angularSpringConstant=5000
    bOverrideAngularDamping=true
    bUseGravityOverride=true
    linearDampingOverride=1
    linearSpringConstant=500
    simulationSpace=RootRelative
    boundBone {boneName=dyn_chain_04}
    componentAppliedLinearAccClamp=<100 100 100>
    componentLinearAccScale=<0.1 0.1 0.1>
    componentLinearVelScale=<0.1 0.1 0.1>
    physicsBodyDefinitions [{
      collisionType=CustomSphere
      sphereCollisionRadius=2
      boundBone {boneName=dyn_chain_04}
      boxExtents=<34 24 24>
      constraintSetup {angularLimitsMax=<0 0 2.5>, angularLimitsMin=<0 -2.5 -5>, angularTarget=<1 0 0>}
      localJointOffset=<5 0 0>
     }]
    sphericalLimits [
     {limitRadius=10, drivingBone {boneName=upperarm_l_in}, sphereLocalOffset=<0 0 0>}
     {limitRadius=17, drivingBone {boneName=spine_05_in}, sphereLocalOffset=<0 0 0>}
     {limitRadius=17, drivingBone {boneName=spine_04_in}, sphereLocalOffset=<0 0 0>}
     {limitRadius=17, drivingBone {boneName=spine_03_in}, sphereLocalOffset=<0 0 0>}
     {limitRadius=17, drivingBone {boneName=spine_02_in}, sphereLocalOffset=<0 0 0>}
     {limitRadius=8, drivingBone {boneName=upperarm_l_in}, sphereLocalOffset=<15 0 0>}
    ]
   }
   n26 = Anim Dynamics - Bone: dyn_chain_05(ComponentPose=@n25, Alpha=1, GravityOverride="0,0,-980") {
    angularDampingOverride=0.9
    angularSpringConstant=5000
    bOverrideAngularDamping=true
    bUseGravityOverride=true
    linearDampingOverride=1
    linearSpringConstant=500
    simulationSpace=RootRelative
    boundBone {boneName=dyn_chain_05}
    componentAppliedLinearAccClamp=<100 100 100>
    componentLinearAccScale=<0.1 0.1 0.1>
    componentLinearVelScale=<0.1 0.1 0.1>
    physicsBodyDefinitions [{
      collisionType=CustomSphere
      sphereCollisionRadius=2
      boundBone {boneName=dyn_chain_05}
      boxExtents=<33 23 23>
      constraintSetup {angularLimitsMax=<0 0 2.5>, angularLimitsMin=<0 -2.5 -5>, angularTarget=<1 0 0>}
      localJointOffset=<5 0 0>
     }]
    sphericalLimits [
     {limitRadius=10, drivingBone {boneName=upperarm_l_in}, sphereLocalOffset=<0 0 0>}
     {limitRadius=17, drivingBone {boneName=spine_05_in}, sphereLocalOffset=<0 0 0>}
     {limitRadius=17, drivingBone {boneName=spine_04_in}, sphereLocalOffset=<0 0 0>}
     {limitRadius=17, drivingBone {boneName=spine_03_in}, sphereLocalOffset=<0 0 0>}
     {limitRadius=17, drivingBone {boneName=spine_02_in}, sphereLocalOffset=<0 0 0>}
     {limitRadius=8, drivingBone {boneName=upperarm_l_in}, sphereLocalOffset=<15 0 0>}
    ]
   }
   n27 = Anim Dynamics - Bone: dyn_chain_06(ComponentPose=@n26, Alpha=1, GravityOverride="0,0,-980") {
    angularDampingOverride=0.9
    angularSpringConstant=5000
    bOverrideAngularDamping=true
    bUseGravityOverride=true
    linearDampingOverride=1
    linearSpringConstant=500
    simulationSpace=RootRelative
    boundBone {boneName=dyn_chain_06}
    componentAppliedLinearAccClamp=<100 100 100>
    componentLinearAccScale=<0.1 0.1 0.1>
    componentLinearVelScale=<0.1 0.1 0.1>
    physicsBodyDefinitions [{
      collisionType=CustomSphere
      sphereCollisionRadius=2
      boundBone {boneName=dyn_chain_06}
      boxExtents=<32 22 22>
      constraintSetup {angularLimitsMax=<0 0 2.5>, angularLimitsMin=<0 -2.5 -5>, angularTarget=<1 0 0>}
      localJointOffset=<5 0 0>
     }]
    sphericalLimits [
     {limitRadius=10, drivingBone {boneName=upperarm_l_in}, sphereLocalOffset=<0 0 0>}
     {limitRadius=17, drivingBone {boneName=spine_05_in}, sphereLocalOffset=<0 0 0>}
     {limitRadius=17, drivingBone {boneName=spine_04_in}, sphereLocalOffset=<0 0 0>}
     {limitRadius=17, drivingBone {boneName=spine_03_in}, sphereLocalOffset=<0 0 0>}
     {limitRadius=17, drivingBone {boneName=spine_02_in}, sphereLocalOffset=<0 0 0>}
     {limitRadius=8, drivingBone {boneName=upperarm_l_in}, sphereLocalOffset=<15 0 0>}
    ]
   }
   n28 = Anim Dynamics - Bone: dyn_chain_07(ComponentPose=@n27, Alpha=1, GravityOverride="0,0,-980") {
    angularDampingOverride=0.9
    angularSpringConstant=5000
    bOverrideAngularDamping=true
    bUseGravityOverride=true
    linearDampingOverride=1
    linearSpringConstant=500
    simulationSpace=RootRelative
    boundBone {boneName=dyn_chain_07}
    componentAppliedLinearAccClamp=<100 100 100>
    componentLinearAccScale=<0.1 0.1 0.1>
    componentLinearVelScale=<0.1 0.1 0.1>
    physicsBodyDefinitions [{
      collisionType=CustomSphere
      sphereCollisionRadius=2
      boundBone {boneName=dyn_chain_07}
      boxExtents=<31 21 21>
      constraintSetup {angularLimitsMax=<0 0 2.5>, angularLimitsMin=<0 -2.5 -5>, angularTarget=<1 0 0>}
      localJointOffset=<5 0 0>
     }]
    sphericalLimits [
     {limitRadius=10, drivingBone {boneName=upperarm_l_in}, sphereLocalOffset=<0 0 0>}
     {limitRadius=17, drivingBone {boneName=spine_05_in}, sphereLocalOffset=<0 0 0>}
     {limitRadius=17, drivingBone {boneName=spine_04_in}, sphereLocalOffset=<0 0 0>}
     {limitRadius=17, drivingBone {boneName=spine_03_in}, sphereLocalOffset=<0 0 0>}
     {limitRadius=17, drivingBone {boneName=spine_02_in}, sphereLocalOffset=<0 0 0>}
     {limitRadius=8, drivingBone {boneName=upperarm_l_in}, sphereLocalOffset=<15 0 0>}
    ]
   }
   n29 = Anim Dynamics - Bone: dyn_chain_08(ComponentPose=@n28, Alpha=1, GravityOverride="0,0,-980") {
    angularDampingOverride=0.9
    angularSpringConstant=5000
    bOverrideAngularDamping=true
    bUseGravityOverride=true
    linearDampingOverride=1
    linearSpringConstant=500
    simulationSpace=RootRelative
    boundBone {boneName=dyn_chain_08}
    componentAppliedLinearAccClamp=<100 100 100>
    componentLinearAccScale=<0.1 0.1 0.1>
    componentLinearVelScale=<0.1 0.1 0.1>
    physicsBodyDefinitions [{
      collisionType=CustomSphere
      sphereCollisionRadius=2
      boundBone {boneName=dyn_chain_08}
      boxExtents=<30 20 20>
      constraintSetup {angularLimitsMax=<0 0 2.5>, angularLimitsMin=<0 -2.5 -5>, angularTarget=<1 0 0>}
      localJointOffset=<5 0 0>
     }]
    sphericalLimits [
     {limitRadius=10, drivingBone {boneName=upperarm_l_in}, sphereLocalOffset=<0 0 0>}
     {limitRadius=17, drivingBone {boneName=spine_05_in}, sphereLocalOffset=<0 0 0>}
     {limitRadius=17, drivingBone {boneName=spine_04_in}, sphereLocalOffset=<0 0 0>}
     {limitRadius=17, drivingBone {boneName=spine_03_in}, sphereLocalOffset=<0 0 0>}
     {limitRadius=17, drivingBone {boneName=spine_02_in}, sphereLocalOffset=<0 0 0>}
     {limitRadius=8, drivingBone {boneName=upperarm_l_in}, sphereLocalOffset=<15 0 0>}
    ]
   }
   n30 = Anim Dynamics - Bone: dyn_chain_09(ComponentPose=@n29, Alpha=1, GravityOverride="0,0,-980") {
    angularDampingOverride=0.9
    angularSpringConstant=5000
    bOverrideAngularDamping=true
    bUseGravityOverride=true
    linearDampingOverride=1
    linearSpringConstant=500
    simulationSpace=RootRelative
    boundBone {boneName=dyn_chain_09}
    componentAppliedLinearAccClamp=<100 100 100>
    componentLinearAccScale=<0.1 0.1 0.1>
    componentLinearVelScale=<0.1 0.1 0.1>
    physicsBodyDefinitions [{
      collisionType=CustomSphere
      sphereCollisionRadius=2
      boundBone {boneName=dyn_chain_09}
      boxExtents=<29 19 19>
      constraintSetup {angularLimitsMax=<0 0 2.5>, angularLimitsMin=<0 -2.5 -5>, angularTarget=<1 0 0>}
      localJointOffset=<5 0 0>
     }]
    sphericalLimits [
     {limitRadius=10, drivingBone {boneName=upperarm_l_in}, sphereLocalOffset=<0 0 0>}
     {limitRadius=17, drivingBone {boneName=spine_05_in}, sphereLocalOffset=<0 0 0>}
     {limitRadius=17, drivingBone {boneName=spine_04_in}, sphereLocalOffset=<0 0 0>}
     {limitRadius=17, drivingBone {boneName=spine_03_in}, sphereLocalOffset=<0 0 0>}
     {limitRadius=17, drivingBone {boneName=spine_02_in}, sphereLocalOffset=<0 0 0>}
     {limitRadius=8, drivingBone {boneName=upperarm_l_in}, sphereLocalOffset=<15 0 0>}
    ]
   }
   n31 = Anim Dynamics - Bone: dyn_handle(ComponentPose=@n30, Alpha=1, GravityOverride="0,0,-980") {
    angularDampingOverride=0.9
    angularSpringConstant=5000
    bOverrideAngularDamping=true
    bUseGravityOverride=true
    linearDampingOverride=1
    linearSpringConstant=500
    simulationSpace=RootRelative
    boundBone {boneName=dyn_handle}
    componentAppliedLinearAccClamp=<100 100 100>
    componentLinearAccScale=<0.1 0.1 0.1>
    componentLinearVelScale=<0.1 0.1 0.1>
    physicsBodyDefinitions [{
      collisionType=CustomSphere
      sphereCollisionRadius=2
      boundBone {boneName=dyn_handle}
      boxExtents=<28 18 18>
      constraintSetup {angularLimitsMax=<0 0 2.5>, angularLimitsMin=<0 -2.5 -5>, angularTarget=<1 0 0>}
      localJointOffset=<5 0 0>
     }]
    sphericalLimits [
     {limitRadius=11, drivingBone {boneName=upperarm_l_in}, sphereLocalOffset=<0 0 0>}
     {limitRadius=17, drivingBone {boneName=spine_05_in}, sphereLocalOffset=<0 0 0>}
     {limitRadius=17, drivingBone {boneName=spine_04_in}, sphereLocalOffset=<0 0 0>}
     {limitRadius=17, drivingBone {boneName=spine_03_in}, sphereLocalOffset=<0 0 0>}
     {limitRadius=17, drivingBone {boneName=spine_02_in}, sphereLocalOffset=<0 0 0>}
     {limitRadius=8, drivingBone {boneName=upperarm_l_in}, sphereLocalOffset=<15 0 0>}
    ]
   }
   n32 = Component To Local(ComponentPose=@n31)
   n33 = Get bIsSkyDiving()
   n34 = Anim Dynamics - Bone: dyn_sciccors!(
    ComponentPose=Local To Component(LocalPose=Blend Poses by bool(BlendPose_0=@n32, BlendPose_1=@n21, bActiveValue=@n33))
    Alpha=1
    GravityOverride:Vector="0,0,-980"
   ) -> (Pose) {
    angularDampingOverride=1
    angularSpringConstant=200
    bLinearSpring=true
    bOverrideAngularDamping=true
    bOverrideLinearDamping=true
    bUseGravityOverride=true
    linearDampingOverride=1
    linearSpringConstant=85
    simulationSpace=RootRelative
    boundBone {boneName=dyn_sciccors}
    componentAppliedLinearAccClamp=<1100 1100 1100>
    componentLinearAccScale=<0.5 0.5 0.4>
    componentLinearVelScale=<0.5 0.5 0.4>
    physicsBodyDefinitions [{
      boundBone {boneName=dyn_sciccors}
      boxExtents=<15 15 15>
      constraintSetup {angularLimitsMax=<0 0 0>, angularLimitsMin=<0 -15 -20>, angularTarget=<1 0 0>, linearAxesMax=<4 0 0>}
      localJointOffset=<0.5 0.5 -1>
     }]
   }
   n35 = Anim Dynamics - Bone: dyn_glove!(ComponentPose=@n34, Alpha=1, GravityOverride:Vector="0,0,-980") -> (Pose) {
    angularDampingOverride=1
    angularSpringConstant=150
    bLinearSpring=true
    bOverrideAngularDamping=true
    bOverrideLinearDamping=true
    bUseGravityOverride=true
    linearDampingOverride=1
    linearSpringConstant=85
    simulationSpace=RootRelative
    boundBone {boneName=dyn_glove}
    componentAppliedLinearAccClamp=<1100 1100 1100>
    componentLinearAccScale=<0.5 0.5 0.35>
    componentLinearVelScale=<0.5 0.5 0.35>
    physicsBodyDefinitions [{
      boundBone {boneName=dyn_glove}
      boxExtents=<25 25 25>
      constraintSetup {angularLimitsMax=<0 0 0>, angularLimitsMin=<0 0 0>, angularTarget=<1 0 0>, linearAxesMax=<1.5 0 0>}
      localJointOffset=<1 0.5 -1>
     }]
   }
   n36 = Anim Dynamics - Bone: dyn_tool!(ComponentPose=@n35, Alpha=1, GravityOverride:Vector="0,0,-980") -> (Pose) {
    angularDampingOverride=1
    angularSpringConstant=150
    bLinearSpring=true
    bOverrideAngularDamping=true
    bOverrideLinearDamping=true
    bUseGravityOverride=true
    linearDampingOverride=1
    linearSpringConstant=85
    simulationSpace=RootRelative
    boundBone {boneName=dyn_tool}
    componentAppliedLinearAccClamp=<1100 1100 1100>
    componentLinearAccScale=<0.5 0.5 0.35>
    componentLinearVelScale=<0.5 0.5 0.35>
    physicsBodyDefinitions [{
      boundBone {boneName=dyn_tool}
      boxExtents=<35 35 35>
      constraintSetup {angularLimitsMax=<0 0 0>, angularLimitsMin=<0 0 -20>, angularTarget=<1 0 0>, linearAxesMax=<1 0 0>}
      localJointOffset=<1 1 -1>
     }]
   }
   Output Pose(Result=Component To Local(ComponentPose=@n36)) {layerGroup=None}
   n37 = "Transform (Modify) Bone - Bone: spine_02_in"!(ComponentPose=Local To Component(LocalPose=Input Pose()), Alpha=1, Translation="0,0,0", Rotation="0,0,0") -> (Pose) {
    rotationMode=BMM_Replace
    rotationSpace=BCS_WorldSpace
    translationMode=BMM_Replace
    translationSpace=BCS_WorldSpace
    alphaBoolBlend {blendOption=Linear}
    boneToModify {boneName=spine_02_in}
   }
   n38 = "Transform (Modify) Bone - Bone: spine_03_in"!(ComponentPose=@n37, Alpha=1, Translation="0,0,0", Rotation="0,0,0") -> (Pose) {
    rotationMode=BMM_Replace
    rotationSpace=BCS_WorldSpace
    translationMode=BMM_Replace
    translationSpace=BCS_WorldSpace
    alphaBoolBlend {blendOption=Linear}
    boneToModify {boneName=spine_03_in}
   }
   n39 = "Transform (Modify) Bone - Bone: spine_04_in"!(ComponentPose=@n38, Alpha=1, Translation="0,0,0", Rotation="0,0,0") -> (Pose) {
    rotationMode=BMM_Replace
    rotationSpace=BCS_WorldSpace
    translationMode=BMM_Replace
    translationSpace=BCS_WorldSpace
    alphaBoolBlend {blendOption=Linear}
    boneToModify {boneName=spine_04_in}
   }
   n40 = "Transform (Modify) Bone - Bone: spine_05_in"!(ComponentPose=@n39, Alpha=1, Translation="0,0,0", Rotation="0,0,0") -> (Pose) {
    rotationMode=BMM_Replace
    rotationSpace=BCS_WorldSpace
    translationMode=BMM_Replace
    translationSpace=BCS_WorldSpace
    alphaBoolBlend {blendOption=Linear}
    boneToModify {boneName=spine_05_in}
   }
   n41 = "Transform (Modify) Bone - Bone: upperarm_l_in"!(ComponentPose=@n40, Alpha=1, Translation="0,0,0", Rotation="0,0,0") -> (Pose) {
    rotationMode=BMM_Replace
    rotationSpace=BCS_WorldSpace
    translationMode=BMM_Replace
    translationSpace=BCS_WorldSpace
    alphaBoolBlend {blendOption=Linear}
    boneToModify {boneName=upperarm_l_in}
   }
   Save cached pose 'start pose'(Pose=Component To Local(ComponentPose=@n41)) {cacheName=start pose}
  ]}
 EventGraph {nodes [
   Event BlueprintUpdateAnimation()
   > n1 = Sequence()
   > Set Gravity Override(Gravity Override=SelectVector(A="0,0,100", B="0,0,-980", bPickA=Get bIsSkyDiving()))
   n2 = Get BodyMesh() -> (BodyMesh:Skeletal Mesh Component Object Reference)
   Is Valid(exec=@n1.then_1, InputObject=@n2)
   n3 = BreakTransform(InTransform=GetSocketTransform(self=@n2, InSocketName=spine_02, TransformSpace=RTS_World))
   > Set spine 02 translation(spine 02 translation:Vector=@n3) -> (Output_Get:Vector)
   > Set spine 02 rotation(spine 02 rotation:Rotator=@n3.Rotation) -> (Output_Get:Rotator)
   n4 = BreakTransform(InTransform=GetSocketTransform(self=@n2, InSocketName=spine_03, TransformSpace=RTS_World))
   > Set spine 03 translation(spine 03 translation:Vector=@n4) -> (Output_Get:Vector)
   > Set spine 03 rotation(spine 03 rotation:Rotator=@n4.Rotation) -> (Output_Get:Rotator)
   n5 = BreakTransform(InTransform=GetSocketTransform(self=@n2, InSocketName=spine_04, TransformSpace=RTS_World))
   > Set spine 04 translation(spine 04 translation:Vector=@n5) -> (Output_Get:Vector)
   > Set spine 04 rotation(spine 04 rotation:Rotator=@n5.Rotation) -> (Output_Get:Rotator)
   n6 = BreakTransform(InTransform=GetSocketTransform(self=@n2, InSocketName=spine_05, TransformSpace=RTS_World))
   > Set spine 05 translation(spine 05 translation:Vector=@n6) -> (Output_Get:Vector)
   > Set spine 05 rotation(spine 05 rotation:Rotator=@n6.Rotation) -> (Output_Get:Rotator)
   n7 = BreakTransform(InTransform=GetSocketTransform(self=@n2, InSocketName=upperarm_l, TransformSpace=RTS_World))
   > Set upperarm l translation(upperarm l translation:Vector=@n7) -> (Output_Get:Vector)
   > Set upperarm l rotation(upperarm l rotation:Rotator=@n7.Rotation) -> (Output_Get:Rotator)
   n8 = Event OnCharacterCustomizationLoaded()
   > Set BodyMesh(BodyMesh=Get Array Item with EFort Custom Part Type(Array=@n8.CustomMeshParts))
   n9 = Get Array Item with EFort Custom Part Type(Array:Array of Custom Character Part Object References=@n8.CustomParts) -> (Output:Custom Character Part Object Reference)
   > Set isMale(isMale:bool="Equal (Enum)"(
     A:EFortCustomGender Enum=Get GenderPermitted!(self:Custom Character Part Object Reference=@n9) -> (GenderPermitted:EFortCustomGender Enum=Invalid)
     B:EFortCustomGender Enum=Male
    )) -> (Output_Get:bool)
   TryGetPawnOwner()
  ]}
]
variables [
 BodyMesh {type=SkeletalMeshComponent}
 Gravity Override {type=Vector}
 isMale {type=boolean}
 spine 02 rotation {type=Rotator}
 spine 02 translation {type=Vector}
 spine 03 rotation {type=Rotator}
 spine 03 translation {type=Vector}
 spine 04 rotation {type=Rotator}
 spine 04 translation {type=Vector}
 spine 05 rotation {type=Rotator}
 spine 05 translation {type=Vector}
 upperarm l rotation {type=Rotator}
 upperarm l translation {type=Vector}
]
